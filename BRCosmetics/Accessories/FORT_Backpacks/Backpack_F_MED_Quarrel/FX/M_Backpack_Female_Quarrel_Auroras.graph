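# Unlit, translucent, two-sided material for aurora effects on the female Quarrel backpack, usable on Niagara mesh particles.
# Panning noise, screen-aligned dust and gradients form an intensity that drives curve-atlas color ramps and, with ramp masks, opacity.
# World position offset builds the ribbon from UVs or SplineThicken, adding sine waves and camera offset.
# Scalar lerps select spline, bottom and mesh-fade variants; depth fade applies only on higher-quality deferred paths.
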Material /BRCosmetics/Accessories/FORT_Backpacks/Backpack_F_MED_Quarrel/FX/M_Backpack_Female_Quarrel_Auroras
graph {
 properties {
  bUsedWithNiagaraMeshParticles=true
  blendMode=BLEND_Translucent
  pixelDepthOffsetMode=PDOM_Legacy
  shadingModel=MSM_Unlit
  twoSided=true
  usedShadingModels=Unlit
 }
 nodes [
  n1 = TextureCoordinate()
  n2 = Constant() {r=256}
  n3 = ScalarParameter() {parameterName=Use Spline}
  n4 = LinearInterpolate(A=Constant2Vector() {g=0.05}, B=Constant2Vector() {g=-0.05}, Alpha=@n3) {constB=0}
  n5 = Time()
  n6 = TextureSample(UVs=Add(
    A=MaterialFunctionCall!(TextureResolution=@n2) {materialFunction=/Engine/Functions/Engine_MaterialFunctions02/Texturing/ScreenAlignedPixelToPixelUVs}.Result
    B=Multiply(A=@n4, B=@n5)
   )) {texture=/Game/Global/Textures/Default/PointSplatter/T_Tiling_Dust}
  n7 = Multiply(A=@n6.R, B=@n6.R)
  n8 = TextureCoordinate()
  n9 = OneMinus(None=@n8)
  n10 = Sine(None=ComponentMask(None=TextureCoordinate() {uTiling=0.5, vTiling=0.5}) {g=true})
  n11 = ScalarParameter() {parameterName=Use Bottom}
  n12 = LinearInterpolate(A=LinearInterpolate(A=ComponentMask(None=@n9) {g=true}, B=@n10, Alpha=@n3) {constB=0}, B=@n10, Alpha=@n11) {constB=0}
  n13 = MaterialFunctionCall(Float2=@n1) {materialFunction=&BreakOutFloat2Components}
  n14 = OneMinus(None=@n13.G)
  n15 = CurveAtlasRowParameter(CurveTime=@n14) {
   atlas=/Game/Athena/LinearColorCurves/VFX_CurveAtlas
   curve=/Game/Accessories/FORT_Backpacks/Backpack_F_MED_Quarrel/FX/Curve_Quarrel_Auroras_03
   defaultValue=30
   group=Curve
   parameterName=MaskRamp2
  }
  n16 = ScalarParameter() {parameterName=BottomGlow}
  n17 = Add(A=OneMinus(None=@n8), B=ScalarParameter() {parameterName=BottomClippingOffset})
  n18 = Multiply(
   A=Multiply(A=@n12, B=@n12)
   B=LinearInterpolate(B=ComponentMask(None=@n17) {g=true}, Alpha=ScalarParameter() {parameterName=UseBottomClipping}) {constA=1}
  )
  n19 = TextureCoordinate() {uTiling=4, vTiling=0.1}
  n20 = TextureCoordinate() {uTiling=2}
  n21 = Constant() {r=0.5}
  n22 = LinearInterpolate(
   A=MaterialFunctionCall(UVs=LinearInterpolate(A=@n19, B=@n20, Alpha=@n3) {constB=0}, "Rotation Angle (0-1)"=@n21) {materialFunction=&CustomRotator}.Rotated Values
   B=TextureCoordinate() {uTiling=2, vTiling=0.1}
   Alpha=@n11
  ) {constB=0}
  n23 = TextureCoordinate() {uTiling=2, vTiling=0.1}
  n24 = LinearInterpolate(
   A=MaterialFunctionCall(UVs=LinearInterpolate(A=@n23, B=@n20, Alpha=@n3) {constB=0}, "Rotation Angle (0-1)"=@n21) {materialFunction=&CustomRotator}.Rotated Values
   B=TextureCoordinate() {uTiling=2, vTiling=0.1}
   Alpha=@n11
  ) {constB=0}
  n25 = Multiply(
   A=TextureSample(UVs=Panner(Coordinate=@n22) {speedX=0.2}) {texture=/Game/Effects/Fort_Effects/Textures/Power_Source/T_Power_Source_Sphere_Texture}.R
   B=TextureSample(UVs=Panner(Coordinate=@n24) {speedX=-0.05}) {texture=&T_Inky_Smoke_Tile}.R
  )
  n26 = Add(
   A=LinearInterpolate(A=Multiply(A=@n15.A, B=@n16) {constB=0.5}, Alpha=ScalarParameter() {parameterName=Use Spline}) {constB=0}
   B=Multiply(A=Multiply(A=@n18, B=@n25), B=ScalarParameter() {defaultValue=5, parameterName=TexMultiply}) {constB=5}
  )
  n27 = ScalarParameter() {parameterName=ColorAdd}
  n28 = Add(A=Multiply(A=Multiply(A=@n7) {constB=7}, B=@n12), B=Saturate(None=Add(A=@n26, B=@n27) {constB=0.05}))
  Add(A=@n1, B=@n28)
  n29 = VertexNormalWS()
  n30 = Transform(None=Constant3Vector() {constant=<rgba 0 1 0 0>})
  n31 = Normalize(VectorInput=CrossProduct(A=@n29, B=@n30))
  n32 = TextureCoordinate()
  n33 = MaterialFunctionCall(Float2=Clamp(None=ConstantBiasScale(None=@n32) {bias=-0.5, scale=2}) {minDefault=-1}) {materialFunction=&BreakOutFloat2Components}
  n34 = VectorParameter() {parameterName="Size&&SineWave", channelNames {a=UpDown, b=FrontBack, g=Width, r=Length}, defaultValue=<rgba 100 50 0 0>}
  n35 = Multiply(A=Multiply(A=@n30, B=@n33.G) {constB=50}, B=@n34.Width) {constB=100}
  n36 = Add(A=Multiply(A=Multiply(A=@n31, B=@n33.R) {constB=50}, B=@n34.Length) {constB=100}, B=@n35)
  n37 = VectorParameter() {parameterName=Spline, defaultValue=<rgba 5 5 0 0>}
  n38 = MaterialFunctionCall(WidthBase=@n37.R, WidthTip=@n37.R, Expand U or V UV Channel=StaticBool()) {materialFunction=&SplineThicken}
  n39 = ScalarParameter() {parameterName=Use Spline}
  n40 = LinearInterpolate(A=@n29, B=@n30, Alpha=ScalarParameter() {parameterName=Use Bottom})
  n41 = ComponentMask(None=TextureCoordinate()) {r=true}
  n42 = VectorParameter() {parameterName="UpDown&&FrontBack", channelNames {a=Pan, b=Tiling, g=Pan, r=Tiling}, defaultValue=<rgba 1 1 0 0>}
  n43 = Time()
  n44 = Add(A=Add(A=Multiply(A=@n41, B=@n42.Tiling), B=Multiply(A=@n42.Pan, B=@n43)), B=ScalarParameter() {parameterName=PanOffset1})
  n45 = LinearInterpolate(A=Cosine(None=@n44), B=Sine(None=@n44), Alpha=ScalarParameter() {parameterName=Use Spline})
  n46 = Add(A=Multiply(A=@n41, B=@n42.Tiling), B=Multiply(A=@n42.Pan, B=@n43))
  n47 = ScalarParameter() {parameterName=PanOffset2}
  n48 = Multiply(A=LinearInterpolate(A=@n30, B=@n29, Alpha=ScalarParameter() {parameterName=Use Bottom}), B=Sine(None=Add(A=@n46, B=@n47)))
  n49 = Sine(None=ComponentMask(None=TextureCoordinate() {uTiling=0.5, vTiling=0.5}) {r=true})
  n50 = Add(
   A=Multiply(A=@n34.FrontBack, B=Multiply(A=@n40, B=@n45)) {constB=20}
   B=Multiply(A=Multiply(A=@n34.UpDown, B=@n48) {constB=20}, B=Saturate(None=@n49)) {constB=20}
  )
  n51 = ComponentMask(None=TextureCoordinate()) {g=true}
  n52 = Constant() {r=0.2}
  n53 = Constant() {r=0.3}
  n54 = Constant()
  n55 = Constant() {r=1}
  n56 = Multiply(
   A=@n50
   B=Clamp(None=MaterialFunctionCall(Input=@n51, Input Low=@n52, Input High=@n53, Target Low=@n54, Target High=@n55) {materialFunction=&RemapValueRange}.Result)
  )
  n57 = ScalarParameter() {parameterName=Fade WPO over V}
  Add(
   A=Add(A=LinearInterpolate(A=@n36, B=@n38.WorldPositionOffset, Alpha=@n39), B=LinearInterpolate(A=@n50, B=@n56, Alpha=@n57))
   B=MaterialFunctionCall(Offset Amount=ScalarParameter() {parameterName=Camera Offset}) {materialFunction=/Engine/Functions/Engine_MaterialFunctions02/WorldPositionOffset/CameraOffset}.Clamped Camera Offset
  )
  n58 = CurveAtlasRowParameter(CurveTime=LinearInterpolate(A=@n13.R, B=OneMinus(None=@n13.R), Alpha=ScalarParameter() {group=Curve, parameterName=InvetColorRamp})) {
   atlas=&VFX_CurveAtlas
   curve=/Game/Accessories/FORT_Backpacks/Backpack_F_MED_Quarrel/FX/Curve_Quarrel_Auroras_02
   defaultValue=29
   group=Curve
   parameterName=ColorRamp
  }
  n59 = CurveAtlasRowParameter(CurveTime=@n28) {
   atlas=&VFX_CurveAtlas
   curve=/Game/Accessories/FORT_Backpacks/Backpack_F_MED_Quarrel/FX/Curve_Quarrel_Auroras_01
   defaultValue=28
   group=Curve
   parameterName=AurorasColor
  }
  n60 = Add(A=Multiply(A=@n58) {constB=0.5}, B=@n59)
  n61 = TextureSampleParameter2D(UVs=@n28) {
   group=Curve
   parameterName=CurveTex
   samplerType=SAMPLERTYPE_LinearColor
   texture=/Game/Animation/Game/MainPlayer/Emotes/Lifted/FX/A_Curve_Bubbles
  }
  n62 = ScalarParameter() {group=Curve, parameterName=UseCurveTexOverColorRamp}
  n63 = ParticleColor()
  MaterialFunctionCall(Emissive=Multiply(A=LinearInterpolate(A=@n60, B=@n61.RGB, Alpha=@n62), B=@n63.RGB)) {materialFunction=&MF_FX_EmissiveAdjustments}
  n64 = Sine(None=ComponentMask(None=TextureCoordinate() {uTiling=0.5, vTiling=0.5}) {r=true})
  n65 = ScalarParameter() {parameterName=EdgeWidth}
  Multiply(A=Saturate(None=Add(A=@n64, B=@n65) {constB=0.5}), B=@n35)
  NamedRerouteDeclaration(None=@n38.Normal) {name=SplineThickenNormal, nodeColor=<rgba 0.7578 0 1 1>}
  NamedRerouteUsage()
  Saturate(None=@n33.G)
  n66 = CurveAtlasRowParameter(CurveTime=@n13.R) {
   atlas=&VFX_CurveAtlas
   curve=&Curve_Quarrel_Auroras_03
   defaultValue=30
   group=Curve
   parameterName=MaskRamp1
  }
  n67 = CurveAtlasRowParameter(CurveTime=LinearInterpolate(A=@n13.R, B=OneMinus(None=@n13.R), Alpha=ScalarParameter() {group=Curve, parameterName=InvetMaskRamp})) {
   atlas=&VFX_CurveAtlas
   curve=&Curve_Quarrel_Auroras_03
   defaultValue=30
   group=Curve
   parameterName=MaskRamp1
  }
  n68 = LinearInterpolate(A=@n66.B, B=Multiply(A=@n67.G, B=@n66.B), Alpha=ScalarParameter() {group=Curve, parameterName=UseMaskRampG})
  n69 = CurveAtlasRowParameter(CurveTime=@n14) {
   atlas=&VFX_CurveAtlas
   curve=&Curve_Quarrel_Auroras_03
   defaultValue=30
   group=Curve
   parameterName=MaskRamp2
  }
  n70 = Sine(None=MaterialFunctionCall(Float2=TextureCoordinate() {uTiling=0.5, vTiling=0.5}) {materialFunction=&BreakOutFloat2Components}.G)
  n71 = Multiply(
   A=@n28
   B=LinearInterpolate(A=Multiply(A=@n68, B=@n69.R), B=Multiply(A=@n66.B, B=@n70), Alpha=ScalarParameter() {parameterName=Use Spline}) {constB=0}
  )
  n72 = Multiply(
   A=Multiply(A=Multiply(A=@n63.A, B=@n71), B=ScalarParameter() {defaultValue=0.7, parameterName=FinalMaskMultiply}) {constB=2}
   B=MaterialFunctionCall(Distance=Constant() {r=64}) {materialFunction=&Near_Camera_Fade}.Result
  )
  n73 = Abs(None=DotProduct(A=CameraVectorWS(), B=VertexNormalWS()))
  n74 = DotProduct(
   A=MaterialFunctionCall() {materialFunction=&CameraDirectionVector}.Result
   B=VertexTangentWS()
  )
  n75 = LinearInterpolate(A=Multiply(A=@n73, B=@n73), B=OneMinus(None=Abs(None=@n74)), Alpha=ScalarParameter() {parameterName=Use SplineThicken Normals for MeshFade})
  n76 = LinearInterpolate(A=@n72, B=Multiply(A=@n72, B=Saturate(None=@n75)), Alpha=ScalarParameter() {parameterName=Use MeshFade})
  n77 = DepthFade(Opacity=@n76, FadeDistance=ScalarParameter() {parameterName=FadeDistance})
  n78 = QualitySwitch(Default=@n77, Low=@n76, High=@n77, Medium=@n76, Epic=@n77)
  ShadingPathSwitch(Default=@n78, Deferred=@n78, Forward=@n76, Mobile=@n76)
 ]
}
