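# Niagara system for the female Quarrel backpack cosmetic that renders aurora meshes with an accompanying glow and light.
# The Auroras emitter burst-spawns long-lived particles drawn by seven mesh renderers, computing several offset positions per particle and scaling color by time of day and front-end versus gameplay.
# The Light emitter drives a light and two sprite renderers with color curves cycled by engine time, dimming glow alpha by time of day.
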NiagaraSystem /BRCosmetics/Accessories/FORT_Backpacks/Backpack_F_MED_Quarrel/FX/NS_Backpack_Female_Quarrel_Auroras
dependencies {usedDataInterfaces [&NiagaraDataInterfaceColorCurve], usedDynamicInputs [], usedModules [&SystemState]}
emitters {
 Auroras {
  emitterUpdateScript [
   &EmitterState(Life Cycle Mode=::NewEnumerator1, Loop Behavior=::NewEnumerator1, Loop Duration Mode=::NewEnumerator1, UseLoopDelay=false)
   &SpawnBurst_Instantaneous(Use Spawn Probability=false, UseLoopCountLimit=false)
  ]
  particleSpawnScript [
   &InitializeParticle(Mesh Scale Mode=::NewEnumerator3, UsePositionOffset=false, Write Material Random=true)
   &InitialMeshOrientation(Rotation=<0.25 0 0>, Use Orientation Vector=true, Use Rotation Vector=true)
   &SystemLocation(MaskSpawnGroup=false, Offset=<0 -7 10>)
   /SetVariables(
    Particles.Mesh2Position:Vector3f=&Add_Vector(A=Linked() {linkedVariable {name=Particles.Position, type=NiagaraPosition}}, B=<0 0 5>)
    Particles.Mesh3Position:Vector3f=&Add_Vector(A=Linked() {linkedVariable {name=Particles.Position, type=NiagaraPosition}}, B=<5 2 5>)
   ) {isSetParameters=true}
   /SetVariables(Particles.SplinePosition:Vector3f=&Add_Vector(A=Linked() {linkedVariable {name=Particles.Position, type=NiagaraPosition}}, B=<0 2 -5>)) {isSetParameters=true}
   /SetVariables(
    Particles.BottomPosition1:Vector3f=&Add_Vector(A=Linked() {linkedVariable {name=Particles.Position, type=NiagaraPosition}}, B=<0 0 -10>)
    Particles.BottomPosition2:Vector3f=&Add_Vector(A=Linked() {linkedVariable {name=Particles.Position, type=NiagaraPosition}}, B=<0 0 -5>)
    Particles.BottomPosition3:Vector3f=&Add_Vector(A=Linked() {linkedVariable {name=Particles.Position, type=NiagaraPosition}}, B=<5 2 -8>)
   ) {isSetParameters=true}
   &CameraOffset(Camera Offset Amount=10)
  ]
  particleUpdateScript [
   &ParticleState(Kill Particles When Lifetime Has Elapsed=false, Let Infinitely Lived Particles Die When Emitter Deactivates=true)
   &ScaleColor(
    Scale Alpha=&FNTimeOfDayScalar(Night Scale=0.5, Morning Scale=0.5, Evening Scale=0.5)
    Scale RGB=&VectorFromFloat(Value=&FNTimeOfDayScalar(Night Scale=1, Morning Scale=0.5, Evening Scale=0.5))
    ScaleA=true
    ScaleRGB=true
   )
   &ScaleColor(
    Scale Alpha=&FN_FrontEndFloatModifier(Gameplay=1, Front End=0.7)
    Scale RGB=&VectorFromFloat(Value=&FN_FrontEndFloatModifier(Gameplay=1, Front End=1.7))
    ScaleA=true
    ScaleRGB=true
   ) {module=ScaleColor001}
  ]
  rendererClasses [&NiagaraMeshRendererProperties]
  renderers [
   {rendererClass=&NiagaraMeshRendererProperties, rendererIndex=0}
   {rendererClass=&NiagaraMeshRendererProperties, rendererIndex=1}
   {rendererClass=&NiagaraMeshRendererProperties, rendererIndex=2}
   {rendererClass=&NiagaraMeshRendererProperties, rendererIndex=3}
   {rendererClass=&NiagaraMeshRendererProperties, rendererIndex=4}
   {rendererClass=&NiagaraMeshRendererProperties, rendererIndex=5}
   {rendererClass=&NiagaraMeshRendererProperties, rendererIndex=6}
  ]
 }
 Light {
  emitterUpdateScript [
   &EmitterState(Life Cycle Mode=::NewEnumerator1, Loop Behavior=::NewEnumerator1, Loop Duration Mode=::NewEnumerator1, UseLoopDelay=false)
   &SpawnBurst_Instantaneous(Use Spawn Probability=false, UseLoopCountLimit=false)
  ]
  particleSpawnScript [
   &InitializeParticle(Color=<rgba 0.1996 1 0.4843 1>, Sprite Size=<30 50>, Sprite Size Mode=::NewEnumerator4, UsePositionOffset=false, Write Material Random=true)
   &SystemLocation(MaskSpawnGroup=false, Offset=<0 -17 2>)
   &CameraOffset(Camera Offset Amount=10)
  ]
  particleUpdateScript [
   /SetVariables(
    Particles.GlowColor=&MultiplyLinearColor(
     Linear Color=&ColorFromCurve(
      ColorCurve=DataInterface() {propertyValues {
        AlphaCurve {keys [<at 0 0.5>]}
        BlueCurve {keys [
          <at 0 1>
          {arriveTangent=-1.6317, interpMode=RCIM_Cubic, leaveTangent=-1.6317, time=0.1703, value=0.1979}
          <at 0.4979 0.1875 RCIM_Cubic>
          <at 0.8238 0.1979>
          <at 1 1>
         ]}
        GreenCurve {keys [
          <at 0 0.8666>
          {arriveTangent=-0.4936, interpMode=RCIM_Cubic, leaveTangent=-0.4936, time=0.1703, value=1}
          <at 0.4979 0.6208 RCIM_Cubic>
          <at 0.8238 1>
          <at 1 0.8666>
         ]}
        RedCurve {keys [
          <at 0 0.1996>
          {arriveTangent=1.6074, interpMode=RCIM_Cubic, leaveTangent=1.6074, time=0.1703, value=0.1406}
          <at 0.4979 1 RCIM_Cubic>
          <at 0.8238 0.1406>
          <at 1 0.1996>
         ]}
       }}
      CurveIndex=&Frac_Float(Float=&Multiply_Float(A=$Engine.Time, B=0.2))
     )
     Scale Factor=&MakeFloat4FromVectorAndFloat(
      "Vector (RGB)"=&VectorFromFloat(Value=1)
      "Float (Alpha)"=&FNTimeOfDayScalar(Night Scale=0.5, Morning Scale=0.5, Evening Scale=0.5)
     )
    )
    Particles.GlowPosition:Vector3f=&Add_Vector(A=Linked() {linkedVariable {name=Particles.Position, type=NiagaraPosition}}, B=<0 11 0>)
    Particles.LightColor=&MultiplyLinearColor(
     Linear Color=&ColorFromCurve(
      ColorCurve=DataInterface() {propertyValues {
        AlphaCurve {keys [<at 0 1 RCIM_Cubic>]}
        BlueCurve {keys [
          <at 0 5>
          {arriveTangent=-8.1586, interpMode=RCIM_Cubic, leaveTangent=-8.1586, time=0.1703, value=0.9896}
          <at 0.4979 0.9375 RCIM_Cubic>
          <at 0.8238 0.9896>
          <at 1 5>
         ]}
        GreenCurve {keys [
          <at 0 4.333>
          {arriveTangent=-2.4678, interpMode=RCIM_Cubic, leaveTangent=-2.4678, time=0.1703, value=5}
          <at 0.4979 3.1042 RCIM_Cubic>
          <at 0.8238 5>
          <at 1 4.333>
         ]}
        RedCurve {keys [
          <at 0 0.998>
          {arriveTangent=8.037, interpMode=RCIM_Cubic, leaveTangent=8.037, time=0.1703, value=0.7031}
          <at 0.4979 5 RCIM_Cubic>
          <at 0.8238 0.7031>
          <at 1 0.998>
         ]}
       }}
      CurveIndex=&Frac_Float(Float=&Multiply_Float(A=$Engine.Time, B=0.2))
     )
     Scale Factor=<0.5 0.5 0.5 1>
    )
   ) {isSetParameters=true}
   &ParticleState(Kill Particles When Lifetime Has Elapsed=false, Let Infinitely Lived Particles Die When Emitter Deactivates=true)
   &SpriteFacingAndAlignment(Sprite Alignment=<1 0 0>, Sprite Facing=<0 1 0>, WriteAlignment=true, WriteFacing=true)
   /SetVariables(Particles.Glow02Facing:Vector3f=<0 0 1>) {isSetParameters=true}
  ]
  rendererClasses [&NiagaraLightRendererProperties, &NiagaraSpriteRendererProperties]
  renderers [
   {rendererClass=&NiagaraLightRendererProperties, rendererIndex=0}
   {rendererClass=&NiagaraSpriteRendererProperties, rendererIndex=1}
   {rendererClass=&NiagaraSpriteRendererProperties, rendererIndex=2}
  ]
 }
}
summary {
 systemName=NS_Backpack_Female_Quarrel_Auroras
 emitters [
  {emitterName=Auroras, rendererClasses [&NiagaraMeshRendererProperties]}
  {emitterName=Light, rendererClasses [&NiagaraLightRendererProperties, &NiagaraSpriteRendererProperties]}
 ]
 userVariables [
  User.bIsFrontEnd {type=NiagaraBool}
  User.bIsFrontEndPreview {type=NiagaraBool}
 ]
}
